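; Rock prop building actor (Prop_Rock_06) that varies its material appearance per instance.
; On construction, when cosmetic events can execute, it sets editor-only instance material parameters for two random color vectors from its color variables.
; It also sets a rock lerp scalar from a location-seeded random float, so the blend value follows the actor's placement.
; BeginPlay only calls the parent.
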(event EventBeginPlay
  (|CallParentFunction))

(fn ConstructionScript
  (|CallParentFunction)
  (Networking|CanExecuteCosmeticEvents
    (:True
      (EditorOnlyParameters|EditorOnlySetInstanceMaterialVectorParameter "RandomColor_A" (Variables|Colors|GetRandomColor_A))
      (EditorOnlyParameters|EditorOnlySetInstanceMaterialVectorParameter "RandomColor_B" (Variables|Colors|GetRandomColor_B))
      (EditorOnlyParameters|EditorOnlySetInstanceMaterialScalarParameter "RockRandomLerpValue" (Class|EnvironmentArtFuncLib|RandomFloatbyLocation (Transformation|GetActorLocation))))))
